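# Editor utility actor for the Material Instance Converter that previews a source and a target material instance on two planes for visual comparison.
# RunMaterialTest stores both instances, assigns them to the planes, shows only the target and resets the view mode.
# TogglePlanes flips each plane's visibility.
# ToggleViewMode cycles the editor viewport through Lit, Unlit and GBuffer visualizations via ValidateViewMode, which wraps the index and updates a text label.
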EditorUtilityBlueprint /Game/NeverCook/EditorUtilities/MaterialInstanceConverter/Assets/EUA_4LayerDisplay
graphs [
 EventGraph {nodes [
   Event BeginPlay()
   Event ActorBeginOverlap()
   Event Tick!() -> (OutputDelegate, then, DeltaSeconds:float)
  ]}
 RunMaterialTest {nodes [
   n1 = RunMaterialTest!() -> (then, MaterialSource:Material Instance Object Reference, MaterialTarget:Material Instance Object Reference) {is=title K2Node_FunctionEntry}
   > Set Material Source(Material Source:Material Instance Object Reference=@n1) -> (Output_Get:Material Instance Object Reference)
   > Set Material Target(Material Target:Material Instance Object Reference=@n1.MaterialTarget) -> (Output_Get:Material Instance Object Reference)
   > n2 = Sequence!(execute) -> (then_0, then_1, then_2:Exec, then_3:Exec)
   > n3 = Is Valid(InputObject=Get Material Source() -> (Material Source:Material Instance Object Reference))
   > n4 = Is Valid(InputObject=Get Material Target() -> (Material Target:Material Instance Object Reference))
   Return Node(execute=[@n4.Is Not Valid, @n3.Is Not Valid])
   SetMaterial(
    execute=@n2.then_1
    self=Get TargetPlane() -> (TargetPlane:Static Mesh Component Object Reference)
    ElementIndex=0
    Material=Get Material Target()
   )
   > SetMaterial(self=Get SourcePlane() -> (SourcePlane:Static Mesh Component Object Reference), ElementIndex=0, Material=Get Material Source())
   SetVisibility!(execute=@n2.then_2, self:Scene Component Object Reference=Get TargetPlane(), bNewVisibility:bool=true, bPropagateToChildren:bool=false) -> (then)
   > SetVisibility(self=Get SourcePlane(), bNewVisibility=false, bPropagateToChildren=false)
   Set ViewModeIndex(execute=@n2.then_3, ViewModeIndex:int=0) -> (Output_Get:int=0)
   > ValidateViewMode()
  ]}
 TogglePlanes {nodes [
   TogglePlanes() {is=title K2Node_FunctionEntry}
   n1 = Get TargetPlane()
   > SetVisibility(
    self=@n1
    bNewVisibility=NOT Boolean(A=Get bVisible!(self:Scene Component Object Reference=@n1) -> (bVisible:bool))
    bPropagateToChildren=false
   )
   n2 = Get SourcePlane()
   > SetVisibility(self=@n2, bNewVisibility=NOT Boolean(A=Get bVisible(self=@n2)), bPropagateToChildren=false)
   SpawnTransientObject!(execute, ObjectClass:Object Class Reference, Parent:Object Reference) -> (then, SpawnedObject:Object Reference)
   GetAsset()
   GetAssetsByPackageName!(
    execute
    self:Asset Registry Interface=GetAssetRegistry()
    PackageName:Name=None
    bIncludeOnlyOnDiskAssets:bool=false
    bSkipARFilteredAssets:bool=true
   ) -> (then, OutAssetData:Array of Asset Data Structures, ReturnValue:bool)
  ]}
 ToggleViewMode {nodes [
   ToggleViewMode() {is=title K2Node_FunctionEntry}
   > Increment Int(Value=Get ViewModeIndex() -> (ViewModeIndex:int=0))
   > ValidateViewMode()
  ]}
 UpdateViewModeLabel {nodes [
   n1 = UpdateViewModeLabel!() -> (then, Label:String) {is=title K2Node_FunctionEntry}
   > Set Text!(
    execute
    self:Text Render Component Object Reference=Get TextRender() -> (TextRender:Text Render Component Object Reference)
    Value:"Text (by ref)"="To Text (String)"(InString=@n1)
   ) -> (then) {is=title K2Node_CallFunction}
  ]}
 UserConstructionScript {nodes [Construction Script()]}
 ValidateViewMode {nodes [
   ValidateViewMode() {is=title K2Node_FunctionEntry}
   > n1 = Switch on Int!(execute:Exec, Selection:int=Get ViewModeIndex() -> (ViewModeIndex=0)) -> (Default:Exec, 0:Exec, 1:Exec, 2:Exec, 3:Exec, 4:Exec, 5:Exec, 6:Exec, 7:Exec)
   SetEditorActiveViewportViewMode!(execute=[@n1.0, @n2.then], Index:EViewModeIndex Enum=VMI_Lit) -> (then)
   > UpdateViewModeLabel!(execute, self, Label:String=Lit) -> (then) {is=title K2Node_CallFunction}
   SetEditorActiveViewportViewMode(execute=@n1.1, Index=VMI_Unlit)
   > UpdateViewModeLabel(Label=Unlit)
   SetEditorActiveViewportViewMode(execute=@n1.2, Index=VMI_VisualizeBuffer)
   > SetEditorViewportVisualizeBuffer!(execute, BufferName:Name=BaseColor) -> (then)
   > UpdateViewModeLabel(Label=BaseColor)
   SetEditorActiveViewportViewMode(execute=@n1.3, Index=VMI_VisualizeBuffer)
   > SetEditorViewportVisualizeBuffer(BufferName=Specular)
   > UpdateViewModeLabel(Label=Specular)
   SetEditorActiveViewportViewMode(execute=@n1.4, Index=VMI_VisualizeBuffer)
   > SetEditorViewportVisualizeBuffer(BufferName=Roughness)
   > UpdateViewModeLabel(Label=Roughness)
   SetEditorActiveViewportViewMode(execute=@n1.5, Index=VMI_VisualizeBuffer)
   > SetEditorViewportVisualizeBuffer(BufferName=Metallic)
   > UpdateViewModeLabel(Label=Metallic)
   SetEditorActiveViewportViewMode(execute=@n1.6, Index=VMI_VisualizeBuffer)
   > SetEditorViewportVisualizeBuffer(BufferName=SubsurfaceColor)
   > UpdateViewModeLabel(Label=SubsurfaceColor)
   SetEditorActiveViewportViewMode(execute=@n1.7, Index=VMI_VisualizeBuffer)
   > SetEditorViewportVisualizeBuffer(BufferName=Opacity)
   > UpdateViewModeLabel(Label=Opacity)
   n2 = Set ViewModeIndex(execute=@n1.Default, ViewModeIndex=0) -> (Output_Get=0)
  ]}
]
variables [
 Material Source {type=MaterialInstance, default=null}
 Material Target {type=MaterialInstance, default=null}
 ViewModeIndex {type=integer}
]
